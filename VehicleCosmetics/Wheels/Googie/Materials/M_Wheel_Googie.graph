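# Masked material for the Googie vehicle cosmetic wheel, adding animated emissive and a rim color to the shared MF_WheelParent_Simple attributes.
# Panning noise, shaped by a power and scale, drives base-color-tinted glow in the rim blue and tire red mask channels, with vertex color red separating tire from rim.
# Rim_ColorD replaces base color in the rim alpha mask.
# The result passes through vehicle damage and de-rez transition functions.
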Material /VehicleCosmetics/Wheels/Googie/Materials/M_Wheel_Googie
graph {
 properties {
  bAutomaticallySetUsageInEditor=false
  bUseMaterialAttributes=true
  bUsedWithMeshDeformer=true
  bUsedWithSkeletalMesh=true
  blendMode=BLEND_Masked
  pixelDepthOffsetMode=PDOM_Legacy
  usedShadingModels=Default Lit
 }
 nodes [
  n1 = MaterialFunctionCall() {materialFunction=&MF_WheelParent_Simple}
  n2 = GetMaterialAttributes(None=@n1.MaterialAttributes) {attributeGetTypes [#1, #2]}
  n3 = NamedRerouteUsage()
  n4 = Multiply(A=@n2.Base Color, B=ComponentMask(None=@n1.RimRGBAMask) {b=true})
  n5 = OneMinus(None=@n1.VertexColorR)
  n6 = ComponentMask(None=@n1.TireRGBAMask) {r=true}
  n7 = LinearInterpolate(
   A=LinearInterpolate(A=@n2.Emissive Color, B=Multiply(A=@n3, B=@n4), Alpha=@n5)
   B=Multiply(A=@n3, B=Multiply(A=@n2.Base Color, B=@n6))
   Alpha=@n1.VertexColorR
  )
  n8 = VectorParameter() {group=5_Special Effects, parameterName=Rim_ColorD, sortPriority=33, defaultValue=<rgba 0.9216 0.0021 0.0021 1>}
  n9 = LinearInterpolate(A=@n2.Base Color, B=@n8.RGB, Alpha=Multiply(A=ComponentMask(None=@n1.RimRGBAMask) {a=true}, B=@n5))
  MaterialFunctionCall(In=MaterialFunctionCall(Material=SetMaterialAttributes!(MaterialAttributes=@n2.MaterialAttributes, Emissive Color=@n7, Base Color=@n9) {attributeSetTypes [#1, #2]}) {materialFunction=&MF_BaseVehicle_Damage}.Result) {materialFunction=&MF_VehicleDeRez_Transition}
  n10 = TextureCoordinate()
  n11 = VectorParameter() {parameterName=EmissiveNoiseParams, channelNames {a=LIM Scale, b=Noise, g=Speed, r=Scale}, defaultValue=<rgba 1 0.17 3 6>}
  n12 = Time()
  n13 = TextureSample(UVs=Panner(Coordinate=Multiply(A=@n10, B=@n11.Scale), Time=Multiply(A=@n11.Speed, B=@n12)) {speedX=1, speedY=1}) {texture=/VehicleCosmetics/Wheels/_SharedTextures/FX/RGBNoise_01}
  n14 = Power(Base=ComponentMask(None=AppendVector(A=@n13.A, B=@n13.A)) {g=true}, Exp=@n11.Noise)
  n15 = Constant() {r=0.03}
  NamedRerouteDeclaration(None=Add(A=Multiply(A=@n14, B=@n11.LIM Scale), B=Multiply(A=@n11.LIM Scale, B=@n15))) {name=AnimEmissive, nodeColor=<rgba 0 0.6953 1 1>}
 ]
}
